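# Niagara system for the drone teleport-out effect (Athena TacElim variant), drawing a character toward the drone at the system origin.
# A system-level SampleSkeletalMesh position feeds all three emitters.
# Cubes_Suck lerps wobbling cyan mesh and sprite particles from that point to the origin over their lifetime, while Digitizer_Lines spans both with ribbons.
# Drone_Lights aims beam sprites and meshes along that axis, adding a time-of-day-scaled flare and flickering light via spawn-group masking.
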NiagaraSystem /FNGameplayCues/Effects/Fort_Effects/Effects/Characters/Generic/ResIn/NS_Drone_Teleport_Out_Athena_TacElim
dependencies {
 usedDataInterfaces [
  &NiagaraDataInterfaceCurve
  &NiagaraDataInterfaceSpriteRendererInfo
  &NiagaraDataInterfaceArrayColor
 ]
 usedDynamicInputs [&FNTimeOfDayScalar, &MakeIntFromFloat]
 usedModules [
  &self:SystemSpawnScript.SystemScriptSource.SystemScriptGraph.NiagaraNodeAssignment_0.SetVariables_#1
  &self:SystemSpawnScript.SystemScriptSource.SystemScriptGraph.NiagaraNodeAssignment_1.SetVariables_#2
  &SystemState
  &SampleSkeletalMesh
 ]
}
emitters {
 Cubes_Suck {
  emitterUpdateScript [
   &EmitterState()
   &SpawnRate(
    SpawnRate=&FloatFromCurve(
     FloatCurve=DataInterface() {propertyValues {Curve {keys [{arriveTangent=-1.9259, interpMode=RCIM_Cubic, leaveTangent=-1.9259, tangentMode=RCTM_User, time=0, value=1}, <at 1 0 RCIM_Cubic>]}}}
     CurveIndex=$Emitter.NormalizedLoopAge
     Scale Curve=50
    )
    Use Spawn Probability=false
   )
  ]
  particleSpawnScript [
   &InitializeParticle(
    Color=&SelectLinearColorFromArray(
     Array Sampling Mode=::NewEnumerator0
     Direct Array Index=?
     Color Selection Array=DataInterface() {propertyValues {ColorData [<rgba 0.0447 0.5548 1 1>, <rgba 0 0.9842 1 1>, <rgba 0.3336 0.7227 1 1>, <rgba 0 0.3241 1 1>]}}
     Random Seed=0
    )
    Lifetime=0.5
    Mesh Scale Mode=::NewEnumerator3
    Mesh Uniform Scale=&Multiply_Float(
     A=&RandomRangeFloat(Minimum=0.75, Maximum=1.5)
     B=&RandomRangeFloat(Minimum=0.75, Maximum=2)
    )
    Sprite Size Mode=::NewEnumerator1
    Uniform Sprite Size Max=7
    Uniform Sprite Size Min=4
    UsePositionOffset=false
    Write Material Random=true
   )
   &InitialMeshOrientation(Use Rotation Vector=false)
  ]
  particleUpdateScript [
   &ParticleState()
   &UpdateMeshOrientation(Rotation Vector=&RandomRangeVector(Minimum=<-1 -1 -1>, Maximum=<1 1 1>))
   /SetVariables(Particles.Position=&LerpPosition(
     A=&AddVectorToPosition(
      Position=&TransformPosition(Position=$System.SampleSkeletalMesh.SampledPosition, Source Space=::World, Destination Space=::Simulation)
      Vector=&MakeVector(
       X=&Waveform(
        "Waveform [1]"=::NewEnumerator3
        "[1] Frequency"=&RandomRangeFloat(Minimum=0.1, Maximum=1)
        "[1] Amplitude Min/Max"=<-50 50>
       )
       Y=&Waveform(
        "Waveform [1]"=::NewEnumerator3
        "[1] Frequency"=&RandomRangeFloat(Minimum=0.1, Maximum=1)
        "[1] Amplitude Min/Max"=<-50 50>
       )
       Z=0
      )
     )
     B=/Niagara/DynamicInputs/Helpers/SimulationPosition()
     Alpha=$Particles.NormalizedAge
    )) {isSetParameters=true}
   &ScaleColor(
    Scale Alpha=&FloatFromCurve(FloatCurve=DataInterface() {propertyValues {Curve {keys [
         {tangentMode=RCTM_User, time=0, value=1}
         {arriveTangent=-0.0405, interpMode=RCIM_Cubic, leaveTangent=-0.0405, tangentMode=RCTM_User, time=0.2, value=1}
         {arriveTangent=-3.4994, interpMode=RCIM_Cubic, leaveTangent=-3.4994, tangentMode=RCTM_User, time=1, value=0}
        ]}}})
    ScaleA=true
    ScaleRGB=true
   )
   &DynamicMaterialParameters(
    Index 0 Param 2=1
    Index 0 Param 3=&FloatFromCurve(
     FloatCurve=DataInterface() {propertyValues {Curve {
        defaultValue=1
        keys [<at 0 0 RCIM_Cubic>, {arriveTangent=1.8616, interpMode=RCIM_Cubic, leaveTangent=1.8616, tangentMode=RCTM_User, time=1, value=1}]
       }}}
     Scale Curve=25
    )
    Index 0 Param 4=1
    Module.Index0Float0Write=true
    Module.Index0Float1Write=true
    Module.Index0Float2Write=true
    Module.Index0Float3Write=true
   )
   &ScaleMeshSize(Scale Factor=&VectorFromFloat(Value=&FloatFromCurve(FloatCurve=DataInterface() {propertyValues {Curve {keys [{interpMode=RCIM_Cubic, leaveTangentWeight=0.7348, tangentMode=RCTM_User, time=0, value=1.5}, <at 1 0.5 RCIM_Cubic>]}}})))
   &ScaleSpriteSize(Uniform Curve Sprite Scale=DataInterface() {propertyValues {Curve {keys [{interpMode=RCIM_Cubic, leaveTangentWeight=0.7348, tangentMode=RCTM_User, time=0, value=1.5}, <at 1 0.5 RCIM_Cubic>]}}})
  ]
  rendererClasses [&NiagaraMeshRendererProperties, &NiagaraSpriteRendererProperties]
  renderers [
   {rendererClass=&NiagaraMeshRendererProperties, rendererIndex=0}
   {rendererClass=&NiagaraSpriteRendererProperties, rendererIndex=1}
  ]
 }
 Digitizer_Lines {
  emitterUpdateScript [
   &EmitterState()
   &SpawnBurst_Instantaneous(Spawn Count=2, Use Spawn Probability=false, UseLoopCountLimit=false)
  ]
  particleSpawnScript [
   &PartitionParticles(
    Normalized Index In Partition Includes One=true
    Number Of Partitions=3
    Particles Per Partition=2
    Partition Mode:ENiagara_PartitionMode_v2=::NewEnumerator1
    Write Partition Index to Ribbon Link Order:ENiagara_AttributeSamplingApplyOutput=::NewEnumerator0
    Write Partition to Ribbon ID=::NewEnumerator0
   )
   &InitializeParticle(Color=<rgba 0.3958 1.2227 1.25 1>, Ribbon Width=3, Ribbon Width Mode=::NewEnumerator1, UsePositionOffset=false, Write Material Random=true)
   &CameraOffset(Camera Offset Amount=20)
  ]
  particleUpdateScript [
   &ParticleState()
   /SetVariables(Particles.Position=&LerpPosition(
     A=&AddVectorToPosition(
      Position=&TransformPosition(Position=$System.SampleSkeletalMesh.SampledPosition, Source Space=::World, Destination Space=::Simulation)
      Vector=&MakeVector(
       X=&Waveform(
        "Waveform [1]"=::NewEnumerator3
        "[1] Frequency"=&RandomRangeFloat(Minimum=1, Maximum=3)
        "[1] Amplitude Min/Max"=<-25 25>
       )
       Y=0
       Z=0
      )
     )
     B=&SimulationPosition()
     Alpha=&ReturnNormalizedExecIndex()
    )) {isSetParameters=true}
   &ScaleRibbonWidth(Ribbon Width Scale=&FloatFromCurve(FloatCurve=DataInterface() {propertyValues {Curve {keys [
         <at 0 0 RCIM_Cubic>
         {interpMode=RCIM_Cubic, tangentMode=RCTM_User, time=0.15, value=1}
         {interpMode=RCIM_Cubic, tangentMode=RCTM_User, time=0.85, value=1}
         <at 1 0 RCIM_Cubic>
        ]}}}))
  ]
  rendererClasses [&NiagaraRibbonRendererProperties]
  renderers [{rendererClass=&NiagaraRibbonRendererProperties}]
 }
 Drone_Lights {
  emitterUpdateScript [
   &EmitterState()
   &SpawnBurst_Instantaneous(Spawn Group=1, Use Spawn Probability=false, UseLoopCountLimit=false)
   &SpawnRate(
    SpawnRate=&FloatFromCurve(
     FloatCurve=DataInterface() {propertyValues {Curve {keys [{arriveTangent=-1.9259, interpMode=RCIM_Cubic, leaveTangent=-1.9259, tangentMode=RCTM_User, time=0, value=1}, <at 1 0 RCIM_Cubic>]}}}
     CurveIndex=$Emitter.NormalizedLoopAge
     Scale Curve=50
    )
    Use Spawn Probability=false
   )
  ]
  particleSpawnScript [
   /SetVariables(Particles.VisibilityTag=&SpawnGroup()) {isSetParameters=true}
   &InitializeParticle(
    Color=&MaskLinearColorBySpawnGroup(
     Masked Linear Color=&RandomRangeLinearColor(Minimum=<rgba 0.1912 0.5906 0.6038 0.25>, Maximum=<rgba 0.6333 1.9563 2 0.25>, Color Channel Mode=::NewEnumerator0)
     Passthrough Linear Color=<rgba 1 1 1 1>
    )
    Lifetime=&MaskFloatBySpawnGroup(Masked Float=0.5, Passthrough Float=1.6)
    Mesh Scale=<1 0.5 0.5>
    Mesh Scale Mode=::NewEnumerator4
    Sprite Size=<150 325>
    Sprite Size Mode=::NewEnumerator4
    UsePositionOffset=false
    Write Material Random=true
   )
   &SkeletalMeshLocation(
    Bone / Socket Index=8
    Bone Sampling Mode=::NewEnumerator4
    MaskSpawnGroup=false
    Orientation Sampling=::NewEnumerator1
    Position Sampling=::NewEnumerator2
    Skeletal Mesh=$User.SkeletalMesh
    Use Sampled Mesh Attribute as Ribbon ID=false
   ) {enabled=false}
   &SubUVAnimation(
    Evaluation Type=::NewEnumerator0
    "Sprite Renderer "=DataInterface() {propertyValues {SpriteRenderer=&self:Drone_Lights_0.NiagaraSpriteRendererProperties_1}}
    SubUV Animation Mode:ENiagara_SubUVLookupModeV2=::NewEnumerator2
    UseEndFrame=false
    UseStartFrame=false
   )
   &DynamicMaterialParameters(
    Index 0 Param 1=&RandomRangeFloat(Minimum=0, Maximum=1)
    Index 0 Param 2=1
    Index 0 Param 3=2
    Index 0 Param 4=1
    Module.Index0Float0Write=true
    Module.Index0Float1Write=false
    Module.Index0Float2Write=false
    Module.Index0Float3Write=false
   )
   /SetVariables() {isSetParameters=true}
   &InitialMeshOrientation(
    Mesh Orientation Mode=::NewEnumerator2
    Orientation Vector=&Subtract_Vector(
     A=&TransformPosition(Position=$System.SampleSkeletalMesh.SampledPosition, Source Space=::World, Destination Space=::Simulation)
     B=&SimulationPosition()
    )
    Use Rotation Vector=false
   )
  ]
  particleUpdateScript [
   &ParticleState()
   &AlignSpriteToMeshOrientation(
    Mesh Orientation Relative Sprite Alignment Vector=&Multiply_Vector(
     A=&Subtract_Vector(
      A=&TransformPosition(Position=$System.SampleSkeletalMesh.SampledPosition, Source Space=::World, Destination Space=::Simulation)
      B=&SimulationPosition()
     )
     B=&MakeVector(X=1, Y=1, Z=&Lerp_Float(A=0.35, B=1.35, Alpha=$Particles.MaterialRandom))
    )
    Orientation Quaternion:Quat4f=$Particles.InitialSpriteRotation
    SetAlignment=true
    SetFacing=true
   )
   &ScaleColor(
    Scale Alpha=&MaskFloatBySpawnGroup(
     Masked Float=&FloatFromCurve(FloatCurve=DataInterface() {propertyValues {Curve {
         defaultValue=1
         keys [
          {interpMode=RCIM_Cubic, tangentMode=RCTM_User, time=0, value=0}
          {interpMode=RCIM_Cubic, tangentMode=RCTM_User, time=0.5, value=1}
          {tangentMode=RCTM_User, time=1, value=0}
         ]
        }}})
     Passthrough Float=&FloatFromCurve(
      FloatCurve=DataInterface() {propertyValues {Curve {keys [
          <at 0 0 RCIM_Cubic>
          <at 0.15 1 RCIM_Cubic>
          {arriveTangent=-3.3406, interpMode=RCIM_Cubic, leaveTangent=-3.3406, tangentMode=RCTM_User, time=1, value=0}
         ]}}}
      Scale Curve=$System.TODScalar
     )
    )
    Scale RGB=&VectorFromFloat(Value=&MaskFloatBySpawnGroup(Masked Float=1, Passthrough Float=$System.TODScalar))
    ScaleA=true
    ScaleRGB=true
   )
   &ScaleSpriteSize(
    Scale Factor=&MakeVector2D(
     X=1
     Y=&FloatFromCurve(FloatCurve=DataInterface() {propertyValues {Curve {keys [
          <at 0 0 RCIM_Cubic>
          {interpMode=RCIM_Cubic, tangentMode=RCTM_User, time=0.15, value=1}
          {interpMode=RCIM_Cubic, tangentMode=RCTM_User, time=0.85, value=1}
          <at 1.0023 0 RCIM_Cubic>
         ]}}})
    )
    Scale Sprite Size Mode=::NewEnumerator2
   )
   /SetVariables(
    Particles.FlareSprite_Camera Offset=25
    Particles.FlareSprite_Color:LinearColor=&MultiplyLinearColor(
     Linear Color=<rgba 6.333 19.5625 20 0.7>
     Scale Factor=&MakeVector4FromVectorAndFloat(
      "Vector (RGB)"=<1 1 1>
      "Float (Alpha)"=&FloatFromCurve(FloatCurve=DataInterface() {propertyValues {Curve {keys [<at 0 0 RCIM_Cubic>,<at 0.15 1 RCIM_Cubic>]}}}, Scale Curve=$System.TODScalar)
     )
    )
    Particles.FlareSprite_DynamParam:Vector4f=&MakeVector4(X=0.2, Y=1, Z=1, W=1)
    Particles.FlareSprite_Size:Vector2f=<100 100>
   ) {isSetParameters=true}
   /SetVariables(
    Particles.LightColor=&MultiplyLinearColor(
     Linear Color=<rgba 0.6333 1.9563 2 1>
     Scale Factor=&MakeVector4FromVectorAndFloat(
      "Vector (RGB)"=<1 1 1>
      "Float (Alpha)"=&FloatFromCurve(
       FloatCurve=DataInterface() {propertyValues {Curve {keys [<at 0 1 RCIM_Cubic>,<at 1 0 RCIM_Cubic>]}}}
       Scale Curve=&Multiply_Float(A=75, B=&Waveform("Waveform [1]"=::NewEnumerator4, "[1] Frequency"=3, "[1] Amplitude Min/Max"=<1 2>))
      )
     )
    )
    Particles.LightExponent=4
    Particles.LightPosition:NiagaraPosition=&AddVectorToPosition(
     Position=&SimulationPosition()
     Vector=&MakeVector(
      X=&Waveform(
       "Waveform [1]"=::NewEnumerator3
       "[1] Frequency"=&RandomRangeFloat(Minimum=1, Maximum=3)
       "[1] Amplitude Min/Max"=<-25 25>
      )
      Y=&Waveform(
       "Waveform [1]"=::NewEnumerator3
       "[1] Frequency"=&RandomRangeFloat(Minimum=1, Maximum=3)
       "[1] Amplitude Min/Max"=<-25 25>
      )
      Z=-150
     )
    )
    Particles.LightRadius=150
   ) {isSetParameters=true}
  ]
  rendererClasses [
   &NiagaraSpriteRendererProperties
   &NiagaraMeshRendererProperties
   &NiagaraLightRendererProperties
  ]
  renderers [
   {rendererClass=&NiagaraSpriteRendererProperties}
   {rendererClass=&NiagaraMeshRendererProperties, rendererIndex=1}
   {rendererClass=&NiagaraSpriteRendererProperties, rendererIndex=2}
   {rendererClass=&NiagaraLightRendererProperties, rendererIndex=3}
  ]
 }
}
summary {
 systemName=NS_Drone_Teleport_Out_Athena_TacElim
 emitters [
  {
   emitterName=Drone_Lights
   rendererClasses [
    &NiagaraSpriteRendererProperties
    &NiagaraMeshRendererProperties
    &NiagaraLightRendererProperties
   ]
  }
  {emitterName=Digitizer_Lines, rendererClasses [&NiagaraRibbonRendererProperties]}
  {emitterName=Cubes_Suck, rendererClasses [&NiagaraMeshRendererProperties, &NiagaraSpriteRendererProperties]}
 ]
 userVariables [
  User.MaxOffsets {
   type=NiagaraDataInterfaceArrayFloat
   defaultValue {
    struct=NiagaraExt_VariableValue_DataInterface
    value {
     dataInterface=&self:NiagaraDataInterfaceArrayFloat_1
     dataInterfaceClass=&NiagaraDataInterfaceArrayFloat
    }
   }
  }
  User.MinOffsets {
   type=NiagaraDataInterfaceArrayFloat
   defaultValue {
    struct=NiagaraExt_VariableValue_DataInterface
    value {
     dataInterface=&self:NiagaraDataInterfaceArrayFloat_0
     dataInterfaceClass=&NiagaraDataInterfaceArrayFloat
    }
   }
  }
  User.SkeletalMesh {
   type=NiagaraDataInterfaceSkeletalMesh
   defaultValue {
    struct=NiagaraExt_VariableValue_DataInterface
    value {dataInterface=&self:NiagaraDataInterfaceSkeletalMesh_1}
   }
  }
 ]
}
