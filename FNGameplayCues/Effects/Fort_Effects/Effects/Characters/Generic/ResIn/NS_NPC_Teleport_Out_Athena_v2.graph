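# Niagara teleport-out effect for NPCs that dissolves a character's skeletal mesh into bluish flares, mesh fragments, shimmer and sparkles.
# Flares burst in four spawn groups, while the other emitters spawn over bone-index ranges, with extra groups gated by normalized loop age.
# Fragments, Shimmer and Sparklies accelerate upward and toward User.PointPosition.
# GPU-simulated Heavy_Fragments instead fall under gravity and collide, losing their noise jitter after impact.
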NiagaraSystem /FNGameplayCues/Effects/Fort_Effects/Effects/Characters/Generic/ResIn/NS_NPC_Teleport_Out_Athena_v2
dependencies {
 usedDataInterfaces [
  &NiagaraDataInterfaceCurve
  &NiagaraDataInterfaceColorCurve
  &NiagaraDataInterfaceVectorCurve
  &NiagaraDataInterfaceArrayColor
  &NiagaraDataInterfaceArrayFloat
 ]
 usedDynamicInputs []
 usedModules [&SystemState]
}
emitters {
 Flares {
  emitterUpdateScript [
   &EmitterState()
   &SpawnBurst_Instantaneous(Use Spawn Probability=false, UseLoopCountLimit=false)
   &SpawnBurst_Instantaneous(Spawn Group=1, Use Spawn Probability=false, UseLoopCountLimit=false) {module=SpawnBurst_Instantaneous001}
   &SpawnBurst_Instantaneous(Spawn Group=2, Use Spawn Probability=false, UseLoopCountLimit=false) {module=SpawnBurst_Instantaneous002}
   &SpawnBurst_Instantaneous(Spawn Group=3, Use Spawn Probability=false, UseLoopCountLimit=false) {module=SpawnBurst_Instantaneous003}
  ]
  particleSpawnScript [
   /SetVariables(Particles.VisibilityTag=&SpawnGroup()) {isSetParameters=true}
   &InitializeParticle(
    Color=&SelectLinearColorFromArray(
     Array Sampling Mode=::NewEnumerator1
     Direct Array Index=&SpawnGroup()
     Color Selection Array=DataInterface() {propertyValues {ColorData [<rgba 0 0 0 0.333>, <rgba 0.2123 3.8438 5 0.1>, <rgba 0.2123 3.8438 5 0.25>, <rgba 0.6721 2.1415 5 0.25>]}}
     Random Seed=?
    )
    Lifetime=&MaskFloatBySpawnGroup(Masked Float=0.2, Passthrough Float=0.25, Spawn Group Mask=3)
    Sprite Size Mode=::NewEnumerator3
    Uniform Sprite Size=&MaskFloatBySpawnGroup(
     Masked Float=&RandomRangeFloat(Minimum=400, Maximum=500)
     Passthrough Float=&MaskFloatBySpawnGroup(Masked Float=600, Passthrough Float=1500, Spawn Group Mask=3)
    )
    UsePositionOffset=false
    Write Material Random=true
   )
   &SkeletalMeshLocation(MaskSpawnGroup=false, Mesh Sampling Type=::NewEnumerator5, Skeletal Mesh=$User.SkeletalMesh, Socket Sampling Mode=::NewEnumerator5)
   &CameraOffset(Camera Offset Amount=&SelectFloatFromArray(
     Direct Array Index=&SpawnGroup()
     Float Selection Array=DataInterface() {propertyValues {FloatData [25,50,100,1]}}
    ))
   &DynamicMaterialParameters(
    Index 0 Param 1=&MaskFloatBySpawnGroup(Masked Float=1, Passthrough Float=0.1)
    Index 0 Param 2=1
    Index 0 Param 3=1
    Index 0 Param 4=1
    Module.Param0WriteEnabled=true
    Module.Param1WriteEnabled=true
    Module.Param2WriteEnabled=true
    Module.Param3WriteEnabled=true
   )
  ]
  particleUpdateScript [
   &ParticleState()
   /SetVariables(Particles.CoronaDynamic:Vector4f=&MakeVector4(
     X=&FloatFromCurve(FloatCurve=DataInterface() {propertyValues {Curve {keys [
          {interpMode=RCIM_Cubic, leaveTangentWeight=0.9723, tangentMode=RCTM_User, time=0, value=2}
          {arriveTangentWeight=0.5831, interpMode=RCIM_Cubic, tangentMode=RCTM_User, time=1, value=0.75}
         ]}}})
     Y=1
     Z=0.6
     W=1
    )) {isSetParameters=true}
   &ScaleColor(
    Scale Alpha=&MaskFloatBySpawnGroup(
     Masked Float=&FloatFromCurve(FloatCurve=DataInterface() {propertyValues {Curve {keys [
          <at 0 0 RCIM_Cubic>
          {arriveTangent=1.925, interpMode=RCIM_Cubic, leaveTangent=1.925, time=0.2, value=0.55}
          {arriveTangent=-0.4615, interpMode=RCIM_Cubic, leaveTangent=-0.4615, time=0.4, value=0.77}
          {arriveTangent=-1.2833, interpMode=RCIM_Cubic, leaveTangent=-1.2833, time=0.85, value=0.25}
          <at 1 0 RCIM_Cubic>
         ]}}})
     Passthrough Float=&MaskFloatBySpawnGroup(
      Masked Float=&FloatFromCurve(FloatCurve=DataInterface() {propertyValues {Curve {keys [<at 0 0 RCIM_Cubic>,<at 0.25 1 RCIM_Cubic>,<at 1 0 RCIM_Cubic>]}}})
      Passthrough Float=0.7
      Spawn Group Mask=3
     )
    )
    Scale RGB=&MaskVectorBySpawnGroup(Masked Vector=<2 2 2>, Passthrough Vector=<1 1 1>, Spawn Group Mask=3)
    ScaleA=true
    ScaleRGB=true
   )
   &ScaleSpriteSize(
    Scale Sprite Size Mode=::NewEnumerator0
    Uniform Scale Factor=&MaskFloatBySpawnGroup(
     Masked Float=&FloatFromCurve(FloatCurve=DataInterface() {propertyValues {Curve {keys [{arriveTangent=1.9842, interpMode=RCIM_Cubic, leaveTangent=1.9842, tangentMode=RCTM_User, time=0, value=0.35}, <at 1 1 RCIM_Cubic>]}}})
     Passthrough Float=&MaskFloatBySpawnGroup(
      Masked Float=&FloatFromCurve(FloatCurve=DataInterface() {propertyValues {Curve {keys [
           {arriveTangent=3310.2024, interpMode=RCIM_Cubic, time=0, value=0}
           {arriveTangent=-15.1945, interpMode=RCIM_Cubic, leaveTangent=-3.0716, tangentMode=RCTM_Break, time=0.15, value=1}
           {interpMode=RCIM_Cubic, leaveTangent=-1.6445, time=1, value=0}
          ]}}})
      Passthrough Float=&MaskFloatBySpawnGroup(
       Masked Float=&FloatFromCurve(FloatCurve=DataInterface() {propertyValues {Curve {keys [
            {arriveTangent=3310.2024, interpMode=RCIM_Cubic, time=0, value=0}
            {arriveTangent=0.7273, interpMode=RCIM_Cubic, leaveTangent=0.7273, time=0.1, value=0.3}
            {arriveTangent=-0.3333, interpMode=RCIM_Cubic, leaveTangent=-0.3333, time=0.55, value=0.4}
            {interpMode=RCIM_Cubic, leaveTangent=-1.6445, time=1, value=0}
           ]}}})
       Passthrough Float=&FloatFromCurve(FloatCurve=DataInterface() {propertyValues {Curve {keys [
            {arriveTangent=3310.2024, interpMode=RCIM_Cubic, time=0, value=0.1}
            {arriveTangent=1.4167, interpMode=RCIM_Cubic, leaveTangent=1.4167, time=0.4, value=0.8}
            {arriveTangent=0.3333, interpMode=RCIM_Cubic, leaveTangent=0.3333, time=0.6, value=0.95}
            {interpMode=RCIM_Cubic, leaveTangent=-1.6445, time=1, value=1}
           ]}}})
       Spawn Group Mask=2
      )
      Spawn Group Mask=1
     )
    )
   )
  ]
  rendererClasses [&NiagaraSpriteRendererProperties]
  renderers [
   {rendererClass=&NiagaraSpriteRendererProperties}
   {rendererClass=&NiagaraSpriteRendererProperties, rendererIndex=1}
   {rendererClass=&NiagaraSpriteRendererProperties, rendererIndex=2}
   {rendererClass=&NiagaraSpriteRendererProperties, rendererIndex=3}
  ]
 }
 Fragments {
  emitterUpdateScript [
   &EmitterState(Life Cycle Mode=::NewEnumerator1, Loop Behavior=::NewEnumerator1, Loop Duration=0.75, UseLoopDelay=false)
   &SpawnRate(SpawnRate=250, Use Spawn Probability=true)
   &SpawnRate(
    Spawn Probability=&MakeCustomFloatFromBool(Bool=&SetBoolByFloatComparison(A=$Emitter.NormalizedLoopAge, B=0.333), True Float=1, False Float=0)
    SpawnGroup=1
    SpawnRate=225
    Use Spawn Probability=true
   ) {module=SpawnRate001}
   &SpawnRate(
    Spawn Probability=&MakeCustomFloatFromBool(Bool=&SetBoolByFloatComparison(A=$Emitter.NormalizedLoopAge, B=0.666), True Float=1, False Float=0)
    SpawnGroup=2
    SpawnRate=150
    Use Spawn Probability=true
   ) {module=SpawnRate002}
  ]
  particleSpawnScript [
   &InitializeParticle(
    Color=&MakeLinearColorFromVectorAndFloat("Vector (RGB)"=<25 25 25>, "Float (Alpha)"=&RandomRangeFloat(Minimum=0.7, Maximum=1))
    Lifetime Max=1
    Lifetime Min=0.25
    Lifetime Mode=::NewEnumerator1
    Mesh Scale Mode=::NewEnumerator1
    Mesh Uniform Scale Max=0.5
    Mesh Uniform Scale Min=0.1
    UsePositionOffset=false
    Write Material Random=true
   )
   &SkeletalMeshLocation(
    Bone / Socket Index=&self:BoneIndexRangebySpawnGroup(
     MaxIndex_SpawnGroup0:NiagaraInt32=2
     MaxIndex_SpawnGroup1:NiagaraInt32=11
     MaxIndex_SpawnGroup2:NiagaraInt32=18
     MinIndex_SpawnGroup0:NiagaraInt32=0
     MinIndex_SpawnGroup1:NiagaraInt32=3
     MinIndex_SpawnGroup2:NiagaraInt32=12
     Spawn Group:NiagaraInt32=&SpawnGroup()
    )
    Bone Sampling Mode=::NewEnumerator4
    MaskSpawnGroup=false
    Skeletal Mesh=$User.SkeletalMesh
   )
   &ShapeLocation(MaskSpawnGroup=false, Override Seed=false, Sphere Radius=10, UseOwnerScale=true)
   &AddVelocity(Velocity=&Multiply_VectorByFloat(Vector=$StackContext.ShapeLocation.ShapeNormal, Float=&RandomRangeFloat(Minimum=1, Maximum=3)))
   &InitialMeshOrientation(Use Rotation Vector=false)
  ]
  particleUpdateScript [
   &ParticleState()
   &UpdateMeshOrientation(
    Rotation Rate=&FloatFromCurve(FloatCurve=DataInterface() {propertyValues {Curve {keys [{arriveTangent=-2.8458, interpMode=RCIM_Cubic, leaveTangent=-2.8458, tangentMode=RCTM_User, time=0, value=1}, <at 1 0.2 RCIM_Cubic>]}}})
    Rotation Vector=<2 0.4 0.1>
   )
   &AccelerationForce(Acceleration=<0 0 300>)
   &PointAttractionForce(
    Attraction Radius=512
    Attraction Strength=100
    Attractor Position:NiagaraPosition=$User.PointPosition
    Kill Within Radius=false
    Module.Use Falloff=true
    UseAttractorOffset=false
   )
   &VectorNoiseForce(
    Force Amount=&RandomRangeFloat(Minimum=1250, Maximum=2500)
    Module.Use Random Noise Time Interval=true
    Random Noise Time Interval=0.025
   )
   &SolveForcesAndVelocity(Clamp Velocity=false, Limit Acceleration=false)
   &ScaleMeshSize(Scale Factor=&VectorFromFloat(Value=&FloatFromCurve(FloatCurve=DataInterface() {propertyValues {Curve {keys [<at 0 0 RCIM_Cubic>,<at 0.25 1.25 RCIM_Cubic>,<at 1 0 RCIM_Cubic>]}}})))
   &ScaleColor(
    Scale Alpha=&FloatFromCurve(FloatCurve=DataInterface() {propertyValues {Curve {keys [<at 0.7 1 RCIM_Cubic>,<at 1 0>]}}})
    Scale RGB=&MakeVectorFromLinearColorRGB(Color=&ColorFromCurve(ColorCurve=DataInterface() {propertyValues {
        AlphaCurve {keys []}
        BlueCurve {keys [<at 0 8 RCIM_Cubic>,<at 0.8 0.25>,<at 1 0 RCIM_Cubic>]}
        GreenCurve {keys [<at 0 1 RCIM_Cubic>,<at 0.8 0.25>,<at 1 0 RCIM_Cubic>]}
        RedCurve {keys [<at 0 1 RCIM_Cubic>,<at 0.8 0.25>,<at 1 0 RCIM_Cubic>]}
       }}))
    ScaleA=true
    ScaleRGB=true
   )
  ]
  rendererClasses [&NiagaraMeshRendererProperties]
  renderers [{rendererClass=&NiagaraMeshRendererProperties, rendererIndex=0}]
 }
 Heavy_Fragments {
  simTarget=GPUComputeSim
  emitterUpdateScript [
   &EmitterState(Life Cycle Mode=::NewEnumerator1, Loop Behavior=::NewEnumerator1, Loop Duration=0.75, UseLoopDelay=false)
   &SpawnRate(SpawnRate=150, Use Spawn Probability=true)
   &SpawnRate(
    Spawn Probability=&MakeCustomFloatFromBool(Bool=&SetBoolByFloatComparison(A=$Emitter.NormalizedLoopAge, B=0.333), True Float=1, False Float=0)
    SpawnGroup=1
    SpawnRate=225
    Use Spawn Probability=true
   ) {module=SpawnRate001}
   &SpawnRate(
    Spawn Probability=&MakeCustomFloatFromBool(Bool=&SetBoolByFloatComparison(A=$Emitter.NormalizedLoopAge, B=0.666), True Float=1, False Float=0)
    SpawnGroup=2
    SpawnRate=150
    Use Spawn Probability=true
   ) {module=SpawnRate002}
  ]
  particleSpawnScript [
   &InitializeParticle(
    Color=&MakeLinearColorFromVectorAndFloat("Vector (RGB)"=<25 25 25>, "Float (Alpha)"=&RandomRangeFloat(Minimum=0.7, Maximum=1))
    Lifetime Max=0.85
    Lifetime Min=0.2
    Lifetime Mode=::NewEnumerator1
    Mesh Scale Mode=::NewEnumerator1
    Mesh Uniform Scale Max=0.5
    Mesh Uniform Scale Min=0.1
    UsePositionOffset=false
    Write Material Random=true
   )
   &SkeletalMeshLocation(
    Bone / Socket Index=&self:BoneIndexRangebySpawnGroup(
     MaxIndex_SpawnGroup0:NiagaraInt32=2
     MaxIndex_SpawnGroup1:NiagaraInt32=11
     MaxIndex_SpawnGroup2:NiagaraInt32=18
     MinIndex_SpawnGroup0:NiagaraInt32=0
     MinIndex_SpawnGroup1:NiagaraInt32=3
     MinIndex_SpawnGroup2:NiagaraInt32=12
     Spawn Group:NiagaraInt32=&SpawnGroup()
    )
    Bone Sampling Mode=::NewEnumerator4
    MaskSpawnGroup=false
    Skeletal Mesh=$User.SkeletalMesh
   )
   &ShapeLocation(MaskSpawnGroup=false, Override Seed=false, Sphere Radius=20, UseOwnerScale=true)
   &CameraOffset(Camera Offset Amount=5)
   &MeshRotationForce(Rotation=<1 1 1>)
   &PointForce(Force Strength=500, Module.Use Falloff=false, Module.UseLinearForce=false)
   &ApplyInitialForces()
   &InitialMeshOrientation(Use Rotation Vector=false)
  ]
  particleUpdateScript [
   &ParticleState()
   &ScaleMeshSize(Scale Factor=&VectorFromFloat(Value=&FloatFromCurve(FloatCurve=DataInterface() {propertyValues {Curve {keys [<at 0 0 RCIM_Cubic>,<at 0.25 1.25 RCIM_Cubic>,<at 1 0 RCIM_Cubic>]}}})))
   &GravityForce()
   &Drag(Drag=&RandomRangeFloat(Minimum=4, Maximum=8), Use Linear Drag=true, Use Rotational Drag=true)
   &Collision(
    Age Colliding Particles=false
    Control Roll On Collision=true
    GPU Collision Type=::NewEnumerator2
    Mesh Dimensions:Vector3f=<21 37 0>
    Module.Control Rest State Via Surface Normal=false
    Module.EnableMaxCollisionCount=false
    Module.Randomize Collision Normal Vector=false
    Radius Calculation Type=::NewEnumerator1
    ScaleRateOfRotation=false
   )
   &VectorNoiseForce(
    Force Amount=&MakeCustomFloatFromBool(
     Bool=$StackContext.HasCollided
     True Float=0
     False Float=&RandomRangeFloat(Minimum=1250, Maximum=2500)
    )
    Module.Use Random Noise Time Interval=true
    Random Noise Time Interval=0.025
   )
   &SolveForcesAndVelocity(Clamp Velocity=false, Limit Acceleration=false)
   &ScaleColor(
    Scale Alpha=&FloatFromCurve(FloatCurve=DataInterface() {propertyValues {Curve {keys [<at 0.7 1 RCIM_Cubic>,<at 1 0>]}}})
    Scale RGB=&MakeVectorFromLinearColorRGB(Color=&ColorFromCurve(ColorCurve=DataInterface() {propertyValues {
        AlphaCurve {keys []}
        BlueCurve {keys [<at 0 8 RCIM_Cubic>,<at 0.8 0.25>,<at 1 0 RCIM_Cubic>]}
        GreenCurve {keys [<at 0 1 RCIM_Cubic>,<at 0.8 0.25>,<at 1 0 RCIM_Cubic>]}
        RedCurve {keys [<at 0 1 RCIM_Cubic>,<at 0.8 0.25>,<at 1 0 RCIM_Cubic>]}
       }}))
    ScaleA=true
    ScaleRGB=true
   )
  ]
  rendererClasses [&NiagaraMeshRendererProperties]
  renderers [{rendererClass=&NiagaraMeshRendererProperties, rendererIndex=0}]
 }
 Shimmer {
  emitterUpdateScript [
   &EmitterState(Life Cycle Mode=::NewEnumerator1, Loop Behavior=::NewEnumerator1, Loop Duration=0.75, UseLoopDelay=false)
   &SpawnRate(SpawnRate=20, Use Spawn Probability=true)
   &SpawnRate(
    Spawn Probability=&MakeCustomFloatFromBool(Bool=&SetBoolByFloatComparison(A=$Emitter.NormalizedLoopAge, B=0.333), True Float=1, False Float=0)
    SpawnGroup=1
    SpawnRate=40
    Use Spawn Probability=true
   ) {module=SpawnRate001}
   &SpawnRate(
    Spawn Probability=&MakeCustomFloatFromBool(Bool=&SetBoolByFloatComparison(A=$Emitter.NormalizedLoopAge, B=0.666), True Float=1, False Float=0)
    SpawnGroup=2
    SpawnRate=20
    Use Spawn Probability=true
   ) {module=SpawnRate002}
  ]
  particleSpawnScript [
   &InitializeParticle(
    Color=&ColorFromCurve(
     ColorCurve=DataInterface() {propertyValues {
       AlphaCurve {keys [<at 0 1 RCIM_Cubic>,<at 1 0.5 RCIM_Cubic>]}
       BlueCurve {keys [<at 0 1 RCIM_Cubic>, {arriveTangent=-0.75, interpMode=RCIM_Cubic, leaveTangent=-0.75, time=0.3489, value=0.7122}, <at 1 0.25 RCIM_Cubic>]}
       GreenCurve {keys [<at 0 1 RCIM_Cubic>, {arriveTangent=-0.75, interpMode=RCIM_Cubic, leaveTangent=-0.75, time=0.3489, value=0.7122}, <at 1 0.25 RCIM_Cubic>]}
       RedCurve {keys [
         <at 0 1.371 RCIM_Cubic>
         {arriveTangent=-1.121, interpMode=RCIM_Cubic, leaveTangent=-1.121, time=0.3489, value=0.5965}
         <at 1 0.25 RCIM_Cubic>
        ]}
      }}
     CurveIndex=$Emitter.LoopedAge
    )
    Lifetime=&FloatFromCurve(FloatCurve=DataInterface() {propertyValues {Curve {keys [<at 0 0.4>,<at 1 0.5>]}}}, CurveIndex=$Emitter.NormalizedLoopAge)
    Sprite Rotation Mode=::NewEnumerator2
    Sprite Size Mode=::NewEnumerator3
    Uniform Sprite Size=250
    UsePositionOffset=false
    Write Material Random=true
   )
   &DynamicMaterialParameters(
    Index 0 Param 1=&RandomRangeFloat(Minimum=0, Maximum=1)
    Index 0 Param 2=&RandomRangeFloat(Minimum=0, Maximum=1)
    Index 0 Param 3=&RandomRangeFloat(Minimum=0, Maximum=1)
    Module.Param0WriteEnabled=true
    Module.Param1WriteEnabled=true
    Module.Param2WriteEnabled=true
    Module.Param3WriteEnabled=true
   )
   &CameraOffset(Camera Offset Amount=&RandomRangeFloat(Minimum=-25, Maximum=25))
   &SkeletalMeshLocation(
    Bone / Socket Index=&self:BoneIndexRangebySpawnGroup(
     MaxIndex_SpawnGroup0:NiagaraInt32=2
     MaxIndex_SpawnGroup1:NiagaraInt32=11
     MaxIndex_SpawnGroup2:NiagaraInt32=18
     MinIndex_SpawnGroup0:NiagaraInt32=0
     MinIndex_SpawnGroup1:NiagaraInt32=3
     MinIndex_SpawnGroup2:NiagaraInt32=12
     Spawn Group:NiagaraInt32=&SpawnGroup()
    )
    Bone Sampling Mode=::NewEnumerator4
    MaskSpawnGroup=false
    Skeletal Mesh=$User.SkeletalMesh
   )
   &ShapeLocation(MaskSpawnGroup=false, Override Seed=false, Sphere Radius=25, UseOwnerScale=true)
   &AddVelocity(Velocity=&Multiply_VectorByFloat(Vector=$StackContext.ShapeLocation.ShapeNormal, Float=&RandomRangeFloat(Minimum=-0.04, Maximum=0.02)))
  ]
  particleUpdateScript [
   &ParticleState()
   &AccelerationForce(Acceleration=&RandomRangeVector(Minimum=<0 0 300>, Maximum=<0 0 500>))
   &PointAttractionForce(
    Attraction Radius=512
    Attraction Strength=100
    Attractor Position:NiagaraPosition=$User.PointPosition
    Kill Within Radius=false
    Module.Use Falloff=true
    UseAttractorOffset=false
   )
   &SolveForcesAndVelocity(Clamp Velocity=false, Limit Acceleration=false)
   &SpriteRotationRate(Rotation Rate=&RandomRangeFloat(Minimum=-11, Maximum=11))
   &ScaleSpriteSize(Uniform Curve Sprite Scale=DataInterface() {propertyValues {Curve {keys [
        {arriveTangent=1.9218, interpMode=RCIM_Cubic, leaveTangent=1.9218, tangentMode=RCTM_User, time=0, value=0}
        {arriveTangent=1, interpMode=RCIM_Cubic, leaveTangent=1, time=0.45, value=0.85}
        <at 1 1 RCIM_Cubic>
       ]}}})
   &ScaleColor(
    Scale Alpha=&FloatFromCurve(FloatCurve=DataInterface() {propertyValues {Curve {keys [
         <at 0 0 RCIM_Cubic>
         {interpMode=RCIM_Cubic, tangentMode=RCTM_User, time=0.09, value=0.99}
         {arriveTangent=-1.0879, interpMode=RCIM_Cubic, leaveTangent=-1.0879, time=0.4777, value=0.63}
         <at 1 0 RCIM_Cubic>
        ]}}})
    Scale RGB=&VectorFromCurve(VectorCurve=DataInterface() {propertyValues {XCurve {keys [<at 0 1>,<at 0.85 1>]}, YCurve {keys [<at 0 1>,<at 0.85 1>]}, ZCurve {keys [<at 0 2>,<at 0.85 1>]}}})
    ScaleA=true
    ScaleRGB=true
   )
  ]
  rendererClasses [&NiagaraSpriteRendererProperties]
  renderers [{rendererClass=&NiagaraSpriteRendererProperties}]
 }
 Sparklies {
  emitterUpdateScript [
   &EmitterState(Life Cycle Mode=::NewEnumerator1, Loop Behavior=::NewEnumerator1, Loop Duration=0.5, UseLoopDelay=false)
   &SpawnRate(SpawnRate=200, Use Spawn Probability=false)
   &SpawnRate(
    Spawn Probability=&MakeCustomFloatFromBool(Bool=&SetBoolByFloatComparison(A=$Emitter.NormalizedLoopAge, B=0.333), True Float=1, False Float=0)
    SpawnGroup=1
    SpawnRate=225
    Use Spawn Probability=true
   ) {module=SpawnRate001}
   &SpawnRate(
    Spawn Probability=&MakeCustomFloatFromBool(Bool=&SetBoolByFloatComparison(A=$Emitter.NormalizedLoopAge, B=0.666), True Float=1, False Float=0)
    SpawnGroup=2
    SpawnRate=150
    Use Spawn Probability=true
   ) {module=SpawnRate002}
  ]
  particleSpawnScript [
   &InitializeParticle(
    Color=<rgba 0.1666 0.9901 1 1>
    Lifetime Max=0.8
    Lifetime Min=0.6
    Lifetime Mode=::NewEnumerator1
    Sprite Rotation Mode=::NewEnumerator2
    Sprite Size Mode=::NewEnumerator3
    Uniform Sprite Size=&FloatFromCurve(FloatCurve=DataInterface() {propertyValues {Curve {keys [<at 0 150>,<at 1 50>]}}}, CurveIndex=$Emitter.NormalizedLoopAge)
    UsePositionOffset=false
    Write Material Random=true
   )
   &SkeletalMeshLocation(
    Bone / Socket Index=&self:BoneIndexRangebySpawnGroup(
     MaxIndex_SpawnGroup0:NiagaraInt32=2
     MaxIndex_SpawnGroup1:NiagaraInt32=11
     MaxIndex_SpawnGroup2:NiagaraInt32=18
     MinIndex_SpawnGroup0:NiagaraInt32=0
     MinIndex_SpawnGroup1:NiagaraInt32=3
     MinIndex_SpawnGroup2:NiagaraInt32=12
     Spawn Group:NiagaraInt32=&SpawnGroup()
    )
    Bone Sampling Mode=::NewEnumerator4
    MaskSpawnGroup=false
    Skeletal Mesh=$User.SkeletalMesh
   )
   &ShapeLocation(MaskSpawnGroup=false, Override Seed=false, Sphere Radius=15, UseOwnerScale=true)
   &AddVelocity(Velocity=&Multiply_VectorByFloat(Vector=$StackContext.ShapeLocation.ShapeNormal, Float=&RandomRangeFloat(Minimum=1, Maximum=5)))
  ]
  particleUpdateScript [
   &ParticleState()
   &ScaleSpriteSize(Scale Factor=&Vector2DFromFloat(Value=&FloatFromCurve(FloatCurve=DataInterface() {propertyValues {Curve {keys [
          <at 0 0.49 RCIM_Cubic>
          {interpMode=RCIM_Cubic, tangentMode=RCTM_User, time=0.157, value=0.526}
          {arriveTangent=-5.4783, interpMode=RCIM_Cubic, leaveTangent=-5.4783, time=0.165, value=1.5}
          {interpMode=RCIM_Cubic, tangentMode=RCTM_User, time=0.18, value=0.4}
          {interpMode=RCIM_Cubic, tangentMode=RCTM_User, time=0.38, value=0.26}
          {arriveTangent=-0.2, interpMode=RCIM_Cubic, leaveTangent=-0.2, time=0.41, value=0.83}
          {interpMode=RCIM_Cubic, tangentMode=RCTM_User, time=0.43, value=0.25}
          {interpMode=RCIM_Cubic, tangentMode=RCTM_User, time=0.689, value=0.24}
          {arriveTangent=-0.7407, interpMode=RCIM_Cubic, leaveTangent=-0.7407, time=0.705, value=1}
          {interpMode=RCIM_Cubic, tangentMode=RCTM_User, time=0.716, value=0.22}
          <at 1 0 RCIM_Cubic>
         ]}}})))
   &AccelerationForce(Acceleration=<0 0 300>)
   &PointAttractionForce(
    Attraction Radius=512
    Attraction Strength=100
    Attractor Position:NiagaraPosition=$User.PointPosition
    Kill Within Radius=false
    Module.Use Falloff=true
    UseAttractorOffset=false
   )
   &VectorNoiseForce(
    Force Amount=&RandomRangeFloat(Minimum=2500, Maximum=5000)
    Module.Use Random Noise Time Interval=true
    Random Noise Time Interval=0.05
   )
   &Drag(Drag=&RandomRangeFloat(Minimum=0.6, Maximum=3), Use Linear Drag=true, Use Rotational Drag=true)
   &SolveForcesAndVelocity(Clamp Velocity=false, Limit Acceleration=false)
   &ScaleColor(
    Scale Alpha=&FloatFromCurve(FloatCurve=DataInterface() {propertyValues {Curve {keys [<at 0.7 1 RCIM_Cubic>,<at 1 0>]}}})
    Scale RGB=&VectorFromCurve(VectorCurve=DataInterface() {propertyValues {
       XCurve {keys [<at 0.1 2 RCIM_Cubic>,<at 0.7 1 RCIM_Cubic>]}
       YCurve {keys [<at 0.1 2 RCIM_Cubic>,<at 0.7 0 RCIM_Cubic>]}
       ZCurve {keys [<at 0.1 2 RCIM_Cubic>,<at 0.7 1 RCIM_Cubic>]}
      }})
    ScaleA=true
    ScaleRGB=true
   )
   &SpriteRotationRate(Rotation Rate=&RandomRangeFloat(Minimum=-360, Maximum=360))
  ]
  rendererClasses [&NiagaraSpriteRendererProperties]
  renderers [{rendererClass=&NiagaraSpriteRendererProperties}]
 }
}
summary {
 systemName=NS_NPC_Teleport_Out_Athena_v2
 emitters [
  {emitterName=Heavy_Fragments, simTarget=GPUComputeSim, rendererClasses [&NiagaraMeshRendererProperties]}
  {emitterName=Fragments, rendererClasses [&NiagaraMeshRendererProperties]}
  {emitterName=Sparklies, rendererClasses [&NiagaraSpriteRendererProperties]}
  {emitterName=Shimmer, rendererClasses [&NiagaraSpriteRendererProperties]}
  {emitterName=Flares, rendererClasses [&NiagaraSpriteRendererProperties]}
 ]
 userVariables [
  User.PointPosition {type=NiagaraPosition, description="", defaultValue {}}
  User.SkeletalMesh {
   type=NiagaraDataInterfaceSkeletalMesh
   defaultValue {
    struct=NiagaraExt_VariableValue_DataInterface
    value {dataInterface=&self:NiagaraDataInterfaceSkeletalMesh_0}
   }
  }
 ]
}
